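; Gameplay ability for the Season 23 Augment tutorial quest that checks whether the owning player has Augments ready to be granted.
; On activation, it gets the augment-granting component from the player's controller and reads its list of available Augments.
; If that list is not empty, it sends a complex custom stat event with the quest source tag to the player's quest manager.
; The ability then ends; the passive activation-failure handler only calls its parent.
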(event Ability|EventFailedToActivatePassiveAbility
  (params (ActorInfo :type Gameplay%20Ability%20Actor%20Info%20Structure))
  (|CallParentFunction ActorInfo))

(event Ability|EventActivateAbility
  (|CallParentFunction)
  (bind _returnvalue (Class|FortControllerComponentAugmentGrantingSystem|GetAugmentGrantingComponentfromController (Fortnite|GetFortPlayerControllerfromActor (Variables|GAVars|Data|GetOwningPlayerPawn))))
  (if (> (Utilities|Array|Length (Class|FortControllerComponentAugmentGrantingSystem|GetAvailableAugmentstoGrant _returnvalue)) 0)
    (Quest|SendComplexCustomStatEvent (Variables|GAVars|Data|GetOwningPlayerQuestManager) 0 (Variables|Default|GetT_QuestSourceTag) "(GameplayTags=)" 1)
    (else
      _))
  (Ability|EndAbility))
